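; BeastGirl Blueprint function library of helpers for a charge mechanic in the Storm Claws gameplay.
; GetChargeAndDurationFromChargeStartedEvent unpacks two numbers packed into one Gameplay Event Data value, using whole-and-remainder division by 100 to yield charge and duration.
; GetMaxCharges supplies the fixed charge limit as a hard-coded constant.
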(fn GetChargeAndDurationFromChargeStartedEvent
  (params (EventData :type Gameplay%20Event%20Data%20Structure%20%28by%20ref%29))
  (bind (_remainder _returnvalue) (Math|Float|Division(WholeandRemainder) (Utilities|Struct|BreakGameplayEventData EventData) 1e+02))
  (return _returnvalue _remainder))

(fn GetMaxCharges
  (return 3.0))
